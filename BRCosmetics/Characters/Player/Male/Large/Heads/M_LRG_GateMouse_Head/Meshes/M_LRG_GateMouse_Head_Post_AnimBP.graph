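# Post-process AnimBlueprint for the GateMouse large male head that counter-rotates head and neck bones, apparently to limit head rotation.
# Each update derives negatively scaled rotations from the head's ref-pose delta against neck_02, skipping this during emotes or the GateMouse idle montage.
# Transform Bone nodes apply them to head, neck_01 and neck_02.
# Their shared alpha eases out while the main instance's usercurve01 exceeds a threshold, otherwise back in.
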AnimBlueprint /BRCosmetics/Characters/Player/Male/Large/Heads/M_LRG_GateMouse_Head/Meshes/M_LRG_GateMouse_Head_Post_AnimBP
target_skeleton=/Game/Characters/Player/Common/Fortnite_Base_Head/Fortnite_Base_Head_Export_Skeleton
graphs [
 AnimGraph {nodes [
   n1 = Get HeadRotReduceAlpha() -> (HeadRotReduceAlpha:float)
   n2 = "Transform (Modify) Bone - Bone: head"!(ComponentPose=Local To Component(LocalPose=Input Pose()), Alpha=@n1, Rotation=Get headRot() -> (headRot:Rotator)) -> (Pose) {
    rotationMode=BMM_Replace
    rotationSpace=BCS_BoneSpace
    translationSpace=BCS_BoneSpace
    alphaBoolBlend {blendInTime=0.1, blendOption=Linear, blendOutTime=0.1}
    boneToModify {boneName=head}
   }
   n3 = Get neckRot() -> (neckRot:Rotator)
   n4 = "Transform (Modify) Bone - Bone: neck_01"!(ComponentPose=@n2, Alpha=@n1, Rotation=@n3) -> (Pose) {
    rotationMode=BMM_Replace
    rotationSpace=BCS_BoneSpace
    translationSpace=BCS_BoneSpace
    alphaBoolBlend {blendInTime=0.1, blendOption=Linear, blendOutTime=0.1}
    boneToModify {boneName=neck_01}
   }
   n5 = "Transform (Modify) Bone - Bone: neck_02"!(ComponentPose=@n4, Alpha=@n1, Rotation=@n3) -> (Pose) {
    rotationMode=BMM_Replace
    rotationSpace=BCS_BoneSpace
    translationSpace=BCS_BoneSpace
    alphaBoolBlend {blendInTime=0.1, blendOption=Linear, blendOutTime=0.1}
    boneToModify {boneName=neck_02}
   }
   Output Pose(Result=Component To Local(ComponentPose=@n5))
  ]}
 EventGraph {nodes [
   n1 = Event BlueprintUpdateAnimation()
   > Set Delta Time X(Delta Time X=@n1.DeltaTimeX)
   > n2 = Sequence()
   n3 = Get bIsPlayingEmote()
   n4 = Get IsIdleMontage() -> (IsIdleMontage:bool)
   Branch(execute=@n2.then_1, Condition=AND Boolean(A=NOR Boolean!(A:bool=@n3, B:bool=@n4) -> (ReturnValue:bool), B=Get LimitOn() -> (LimitOn:bool)))
   n5 = BreakRotator(InRot=GetDeltaTransformFromRefPose!(self=GetOwnerSkeletalMeshComponent(), BoneName=head, BaseName=neck_02) -> (ReturnValue_Location:Vector, ReturnValue_Rotation:Rotator, ReturnValue_Scale:Vector).ReturnValue_Rotation)
   n6 = Remap Float(Input=@n5.Yaw, OldRange=(X=10,Y=-28), NewRange=(X=0,Y=1), Clamp=true) -> (Output=~)
   n7 = Lerp(A=Get zHeadRotPerMin() -> (zHeadRotPerMin:double), B=Get zHeadRotPerMax() -> (zHeadRotPerMax:double), Alpha=@n6)
   n8 = float * float!(A=@n5, B=-1, C:double=Get xHeadRotPer() -> (xHeadRotPer:double)) -> (ReturnValue) {is=title K2Node_PromotableOperator}
   n9 = float * float(A=@n5.Pitch, B=-1, C=Get yHeadRotPer() -> (yHeadRotPer:double))
   > Set headRot(headRot:Rotator=MakeRotator(Roll=@n8, Pitch=@n9, Yaw=float * float(A=@n5.Yaw, B=-1, C=@n7))) -> (Output_Get:Rotator)
   n10 = Lerp(A=Get zHeadRotPerMin(), B=0.5, Alpha=@n6)
   > Set neckRot(neckRot:Rotator=MakeRotator(Roll=@n8, Pitch=@n9, Yaw=float * float(B=-1, C=@n10))) -> (Output_Get:Rotator)
   n11 = GetAnimInstance(self=GetOwnerSkeletalMeshComponent())
   Is Valid(exec=@n2.then_0, InputObject=@n11)
   > Set IsIdleMontage(IsIdleMontage:bool=Montage_IsPlaying(self=@n11, Montage=/BRCosmetics/Animation/Game/MainPlayer/Menu/BR/CustomSkins/_s40/GateMouse_M/GateMouse_Idle_Montage)) -> (Output_Get:bool)
   > n12 = Branch(Condition="float > float"(A=GetCurveValue(self=@n11, CurveName=usercurve01), B=0.1))
   > Set HeadRotReduceAlpha(HeadRotReduceAlpha:float=FInterpTo(Current=Get HeadRotReduceAlpha(), Target=0, DeltaTime=Get Delta Time X(), InterpSpeed=4)) -> (Output_Get:float)
   Set HeadRotReduceAlpha(execute=@n12.else, HeadRotReduceAlpha=FInterpTo(Current=Get HeadRotReduceAlpha(), Target=1, DeltaTime=Get Delta Time X(), InterpSpeed=4))
   Event BlueprintBeginPlay() {is=title K2Node_Event}
   > Set LimitOn(LimitOn:bool=true) -> (Output_Get:bool)
   TryGetPawnOwner()
   float * float!(A, B=Lerp(A=Select(Option 0:double=0, Option 1:double=470, Index=false) -> (ReturnValue:double), B=1, Alpha=20)) -> (ReturnValue)
  ]}
]
variables [
 Delta Time X {type=number}
 headRot {type=Rotator}
 HeadRotReduceAlpha {type=number, default=1}
 IsIdleMontage {type=boolean}
 LimitOn {type=boolean}
 neckRot {type=Rotator}
 xHeadRotPer {type=number, default=0.4}
 yHeadRotPer {type=number, default=0.3}
 zHeadRotPerMax {type=number, default=1.35}
 zHeadRotPerMin {type=number}
]
